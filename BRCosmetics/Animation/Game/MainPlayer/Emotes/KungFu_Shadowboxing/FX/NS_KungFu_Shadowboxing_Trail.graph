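# Niagara effect for the KungFu Shadowboxing emote that draws a short blue-tinted strike trail with a bright flash.
# Two self-timed burst emitters render elongated meshes pulled toward the camera: a brief, intense Trail_Flash, then a slightly delayed, longer-lived Trail.
# Both fade alpha over life and animate a dynamic material parameter, likely sweeping the texture along the mesh.
# Trail also ramps another material parameter late in life, and no user parameters are exposed.
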NiagaraSystem /BRCosmetics/Animation/Game/MainPlayer/Emotes/KungFu_Shadowboxing/FX/NS_KungFu_Shadowboxing_Trail
dependencies {usedDataInterfaces [&NiagaraDataInterfaceCurve], usedDynamicInputs [], usedModules [&SystemState]}
emitters {
 Trail {
  emitterUpdateScript [
   &EmitterState(Life Cycle Mode=::NewEnumerator1, Loop Behavior=::NewEnumerator1, Loop Delay=0.1, Loop Duration=1, UseLoopDelay=true)
   &SpawnBurst_Instantaneous(Use Spawn Probability=false, UseLoopCountLimit=false)
  ]
  particleSpawnScript [&InitializeParticle(Lifetime=0.35, Mesh Scale=<1 1 1.8>, Mesh Scale Mode=::NewEnumerator4, UsePositionOffset=false, Write Material Random=true)]
  particleUpdateScript [
   &ParticleState()
   &Color(Color=&MakeLinearColorFromVectorAndFloat(
     "Vector (RGB)"=<0.5624 0.7165 0.7708>
     "Float (Alpha)"=&ClampFloat(Float=&FloatFromCurve(FloatCurve=DataInterface() {propertyValues {Curve {keys [{arriveTangent=-1, time=0, value=10}, <at 1 0>]}}}))
    ))
   &DynamicMaterialParameters(
    Index 0 Param 1=&FloatFromCurve(FloatCurve=DataInterface() {propertyValues {Curve {keys [
         {arriveTangent=0.0003, time=0, value=0}
         {arriveTangent=0.08, interpMode=RCIM_Cubic, leaveTangent=0.08, time=0.5466, value=0.0003}
         <at 1 0.08>
        ]}}})
    Index 0 Param 2=&FloatFromCurve(FloatCurve=DataInterface() {propertyValues {Curve {keys [<at 0 -0.5>,<at 1 1>]}}})
    Index 0 Param 3=1
    Index 0 Param 4=1
    Module.Param0WriteEnabled=true
    Module.Param1WriteEnabled=true
    Module.Param2WriteEnabled=true
    Module.Param3WriteEnabled=true
   )
   &ScaleColor(
    Color Value To Scale=$Particles.Color
    Scale Alpha=&FloatFromCurve(FloatCurve=DataInterface() {propertyValues {Curve {keys [<at 0 1>,<at 1 0>]}}})
    Scale RGB=<3 3 3>
    ScaleA=true
    ScaleRGB=true
   )
   &CameraOffset(Camera Offset Amount=20)
  ]
  rendererClasses [&NiagaraMeshRendererProperties]
  renderers [{rendererClass=&NiagaraMeshRendererProperties, rendererIndex=0}]
 }
 Trail_Flash {
  emitterUpdateScript [
   &EmitterState(Life Cycle Mode=::NewEnumerator1, Loop Behavior=::NewEnumerator1, Loop Duration=1, UseLoopDelay=true)
   &SpawnBurst_Instantaneous(Use Spawn Probability=false, UseLoopCountLimit=false)
  ]
  particleSpawnScript [&InitializeParticle(Lifetime=0.2, Mesh Scale=<1 1 1.8>, Mesh Scale Mode=::NewEnumerator4, UsePositionOffset=false, Write Material Random=true)]
  particleUpdateScript [
   &ParticleState()
   &Color(Color=&MakeLinearColorFromVectorAndFloat(
     "Vector (RGB)"=<0.129 0.2448 0.3073>
     "Float (Alpha)"=&ClampFloat(Float=&FloatFromCurve(FloatCurve=DataInterface() {propertyValues {Curve {keys [{arriveTangent=-1, time=0, value=50}, <at 1 0>]}}}))
    ))
   &DynamicMaterialParameters(
    Index 0 Param 1=1
    Index 0 Param 2=&FloatFromCurve(FloatCurve=DataInterface() {propertyValues {Curve {keys [<at 0 -0.5>,<at 1 1.5>]}}})
    Index 0 Param 3=1
    Index 0 Param 4=1
    Module.Param0WriteEnabled=true
    Module.Param1WriteEnabled=true
    Module.Param2WriteEnabled=true
    Module.Param3WriteEnabled=true
   )
   &ScaleColor(Color Value To Scale=$Particles.Color, Scale RGB=<20 20 20>, ScaleA=true, ScaleRGB=true)
   &CameraOffset(Camera Offset Amount=20)
  ]
  rendererClasses [&NiagaraMeshRendererProperties]
  renderers [{rendererClass=&NiagaraMeshRendererProperties, rendererIndex=0}]
 }
}
summary {
 systemName=NS_KungFu_Shadowboxing_Trail
 emitters [
  {emitterName=Trail, rendererClasses [&NiagaraMeshRendererProperties]}
  {emitterName=Trail_Flash, rendererClasses [&NiagaraMeshRendererProperties]}
 ]
 userVariables []
}
